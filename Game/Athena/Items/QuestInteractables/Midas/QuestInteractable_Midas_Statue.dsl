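; Midas statue quest interactable that offers a "Search" prompt for an Athena quest.
; Client quest enable/disable events toggle the statue mesh and bInteractable, which gates interaction; it defaults to enabled, so the server always allows interaction.
; Objective completion plays sound, Niagara and reward feedback, then hides the statue after a short delay.
; Quest updates unhide each quest visual whose owning player has completed the map-room objective; debug prints remain.
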(event EventBeginPlay)

(event Custom|OnVisualCreated
  (params (QuestVisual :type Quest%20Visual%20Object%20Reference)))

(event Custom|OnQuestsUpdated
  ;[1 1 1] This chunk is just to hide the honeylid if they haven't interacted with the map
  (for _array_element (Variables|Default|GetQuestVisuals)
    (bind _returnvalue (Class|QuestInteractableComponent|GetQuestManagerforActor (Variables|Default|GetQuestInteractable) (Fortnite|GetFortPlayerControllerfromActor (Actor|GetOwner _array_element))))
    (if (Quest|HasCompletedObjectivewithName _returnvalue 0 "quest_s12_special_bvg_maproom_obj" :!class /Script/FortniteGame.FortQuestManager)
      (Rendering|SetActorHiddenInGame _array_element false))))

(event ClientQuestEnabled(QuestInteractable)
  (Development|PrintString "Quest Enabled" false false "(R=0.136643,G=1.000000,B=0.000000,A=1.000000)" 2e+01)
  (Development|SetHiddeninGame (Variables|Default|GetSM_Midas_Lama_Statue) false false)
  ;[1 1 1] LACELLE:  I set this guy to default enabled, so it will only get turned off on the client.  This makes it appear to work properly, (even though the server will always thing you can interact with it even after you've finished the quest).  Perhaps this screws up for splitscreen, though?
  (Variables|Default|SetbInteractable true))

(event ClientQuestDisabled(QuestInteractable)
  (Development|PrintString "Quest Disabled" false false "(R=1.000000,G=0.000000,B=0.063568,A=1.000000)" 2e+01)
  (Development|SetHiddeninGame (Variables|Default|GetSM_Midas_Lama_Statue) true false)
  (Variables|Default|SetbInteractable false))

(event ClientQuestObjectiveComplete(QuestInteractable)
  (Development|PrintString "Objective Complete" false false "(R=0.136643,G=1.000000,B=0.000000,A=1.000000)")
  (Class|AthenaQuestFunctionLibrary|AthenaPlayQuestRewardFeedback false (Variables|Default|GetSuccessSound) (Variables|Default|GetSuccessNiagara) (Variables|Default|GetSuccessFeedback) self)
  (Variables|Default|SetbInteractable false)
  (Utilities|FlowControl|Delay 0.2)
  (Development|SetHiddeninGame (Variables|Default|GetSM_Midas_Lama_Statue) true false))

(event Interaction|EventBlueprintOnInteract
  (params (InteractingPawn :type Fort%20Pawn%20Object%20Reference) (InteractionBeingAttempted :type EInteractionBeingAttempted%20Enum :default "FirstInteraction"))
  (Development|PrintString "INTERACTION COMPLETE" false false "(R=0.136643,G=1.000000,B=0.000000,A=1.000000)"))

(fn BlueprintCanInteract
  (params (InteractingPawn :type Fort%20Pawn%20Object%20Reference) (InteractionBeingAttempted :type EInteractionBeingAttempted%20Enum :default "FirstInteraction") (InteractionType :type TInteractionType%20Enum :default "IT_NoInteraction"))
  (return (Variables|Default|GetbInteractable)))

(fn BlueprintGetInteractionString
  (params (InteractingPawn :type Fort%20Pawn%20Object%20Reference) (InteractionBeingAttempted :type EInteractionBeingAttempted%20Enum :default "FirstInteraction"))
  (return "NSLOCTEXT(\"[5CCCF3AD47392DA77FB29B92828ADB43]\", \"7E7504FB4E0262D2EA504C9DDE07182B\", \"Search\")"))

(fn ConstructionScript)
